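; Camera modifier for the CherrySmoke dual revolvers dodge that offsets the player's view location and FOV.
; On first update it flags a left dodge from last move input, then picks forward/right offsets by direction and aim state, up/FOV by aim alone, scaled by a curve over Alpha.
; A post-process blend eases with IsActive, which clears when the pawn is downed or dead.
; DisableAndRemove disables it, waits out the fade time, then removes it.
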(event Custom|DisableAndRemove
  (CameraModifier|DisableModifier false)
  (Utilities|FlowControl|Delay (ScalableFloat|GetValueatLevel (Variables|Default|GetHF_AplhaOutTime)))
  (bind _cameraowner (Variables|CameraModifier|GetCameraOwner)
    (:then
      (Utilities|IsValid (Game|Player|GetOwningPlayerController _cameraowner)
        (:Is%20Valid
          (Game|Player|RemoveCameraModifier (Variables|/Script/Engine.PlayerController|GetPlayerCameraManager (Game|Player|GetOwningPlayerController _cameraowner)) self))))
    (:else)))

(fn BlueprintModifyCamera
  (params (DeltaTime :type Float%20%28single-precision%29) (ViewLocation :type Vector :default "0, 0, 0") (ViewRotation :type Rotator :default "0, 0, 0") (FOV :type Float%20%28single-precision%29))
  (if (Variables|Default|GetbFirstTimeThrough)
    (Variables|Default|SetbFirstTimeThrough false)
    (Variables|Default|SetbIsLeftDodge (CallFunction|CheckifLeftDodge)))
  (return (Utilities|Operators|Add ViewLocation (* (Math|Vector|GetForwardVector ViewRotation) (Math|Float|Lerp 0.0 (Math|Float|SelectFloat (Math|Float|SelectFloat (Variables|FwdOffset|GetLeftDodgeAimingForwardOffset) (Variables|FwdOffset|GetLeftDodgeForwardOffset) (Input|IsAimingDownSightsorScope (Fortnite|GetFortPlayerControllerfromActor (Variables|CameraModifier|GetCameraOwner)))) (Math|Float|SelectFloat (Variables|FwdOffset|GetAimingForwardOffset) (Variables|FwdOffset|GetForwardOffset) (Input|IsAimingDownSightsorScope (Fortnite|GetFortPlayerControllerfromActor (Variables|CameraModifier|GetCameraOwner)))) (Variables|Default|GetbIsLeftDodge)) (Math|Curves|GetFloatValue (Variables|Default|GetCurve) (Variables|CameraModifier|GetAlpha)))) (* (Math|Vector|GetRightVector ViewRotation) (Math|Float|Lerp 0.0 (Math|Float|SelectFloat (Math|Float|SelectFloat (Variables|RightOffset|GetLeftDodgeAimingRightOffset) (Variables|RightOffset|GetLeftDodgeRightOffset) (Input|IsAimingDownSightsorScope (Fortnite|GetFortPlayerControllerfromActor (Variables|CameraModifier|GetCameraOwner)))) (Math|Float|SelectFloat (Variables|RightOffset|GetAimingRightOffset) (Variables|RightOffset|GetRightOffset) (Input|IsAimingDownSightsorScope (Fortnite|GetFortPlayerControllerfromActor (Variables|CameraModifier|GetCameraOwner)))) (Variables|Default|GetbIsLeftDodge)) (Math|Curves|GetFloatValue (Variables|Default|GetCurve) (Variables|CameraModifier|GetAlpha)))) (Math|Vector|MakeVector 0.0 0.0 (Math|Float|Lerp 0.0 (Math|Float|SelectFloat (Variables|UpOffset|GetAimingUpOffset) (Variables|UpOffset|GetUpOffset) (Input|IsAimingDownSightsorScope (Fortnite|GetFortPlayerControllerfromActor (Variables|CameraModifier|GetCameraOwner)))) (Math|Curves|GetFloatValue (Variables|Default|GetCurve) (Variables|CameraModifier|GetAlpha))))) ViewRotation (+ FOV (Math|Float|Lerp 0.0 (Math|Float|SelectFloat (Variables|FOVOffset|GetAimingFOVOffset) (Variables|FOVOffset|GetFOVOffset) (Input|IsAimingDownSightsorScope (Fortnite|GetFortPlayerControllerfromActor (Variables|CameraModifier|GetCameraOwner)))) (Math|Curves|GetFloatValue (Variables|Default|GetCurve) (Variables|CameraModifier|GetAlpha))))))

(fn BlueprintModifyPostProcess
  (params (DeltaTime :type Float%20%28single-precision%29))
  (bind _asfort_pawn (Utilities|Casting|CastToFortPawn (Pawn|GetControlledPawn (Game|Player|GetOwningPlayerController (Variables|CameraModifier|GetCameraOwner))))
    (:then
      (if (or (Health|IsDBNO _asfort_pawn) (FortPawn|IsDead _asfort_pawn))
        (Variables|Default|SetIsActive false)
        (else
          (Variables|Default|SetBlendWeight (Math|Interpolation|FInterpTo (Variables|Default|GetBlendWeight) (Math|Float|SelectFloat 1.0 0.0 (Variables|Default|GetIsActive)) (Variables|GetDeltaTime) 4.0))
          (return (Variables|Default|GetBlendWeight) (Variables|Default|GetVictimPostProcess)))))
    (:CastFailed
      (Variables|Default|SetIsActive false))))

(fn CheckIfLeftDodge
  (bind (_x _y) (Math|Vector2D|BreakVector2D (Math|Vector2D|NormalSafe(Vector2D) (Variables|/Script/FortniteGame.FortPlayerController|GetLastPlayerLocalMoveInput (Fortnite|GetFortPlayerControllerfromActor (Variables|CameraModifier|GetCameraOwner))) 1e-05)))
  (return (Math|Float|InRange(Float) _y -1.0 -0.5)))
